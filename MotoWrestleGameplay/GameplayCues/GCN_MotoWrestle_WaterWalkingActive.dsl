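; Looping gameplay cue for MotoWrestle water walking, dressing the target player pawn with freezing-water effects and audio.
; On start it positions two particle components, adds a footstep audio mutator and sound library, and binds jump/landing events driving a "Jump" particle parameter.
; A repeating timer polls sliding state to spawn or fade an ice-slide loop.
; Removal reverses most setup but re-binds the jump event; the footstep handler has no effect.
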(event EventBeginPlay)

(event Collision|EventActorBeginOverlap
  (params (OtherActor :type Actor%20Object%20Reference)))

(event Custom|Start%20Ice%20Sliding%20Audio
  ; On Ice Audio Handling
  (bind _targetpawn (Variables|Default|GetTargetPawn)
    (:then
      (if (Variables|/Script/FortniteGame.FortPawn|GetbIsSliding _targetpawn)
        (if (not (Variables|Default|GetbPlayedSliding))
          (bind _iceslideaudiocomponent (Variables|Default|GetIceSlideAudioComponent)
            (:then
              (Audio|Components|Audio|FadeOut _iceslideaudiocomponent 0.3 0.0 "Linear" :!class /Script/Engine.AudioComponent))
            (:else
              _))
          (bind _returnvalue (Audio|SpawnSoundAttached (Variables|Default|GetIceSlidingLoop) (Variables|/Script/Engine.Character|GetMesh (Variables|Default|GetTargetPawn))))
          (Variables|Default|SetIceSlideAudioComponent _returnvalue)
          (Variables|Default|SetbPlayedSliding true))
        (elif (Variables|Default|GetbPlayedSliding)
          (bind _iceslideaudiocomponent_2 (Variables|Default|GetIceSlideAudioComponent)
            (:then
              (Audio|Components|Audio|FadeOut _iceslideaudiocomponent_2 0.3 0.0 "Linear" :!class /Script/Engine.AudioComponent)
              (Variables|Default|SetbPlayedSliding false))
            (:else)))))
    (:else)))

(event EventOnLoopingStart(EitherVFX-preferthisversion)
  (params (MyTarget :type Actor%20Object%20Reference) (Parameters :type Gameplay%20Cue%20Parameters%20Structure%20%28by%20ref%29) (ParticleComponents :type Array%20of%20FXSystem%20Component%20Object%20References) (AudioComponents :type Array%20of%20Audio%20Component%20Object%20References))
  (bind _output (Utilities|Array|Get(acopy) ParticleComponents 1))
  (bind _output_2 (Utilities|Array|Get(acopy) ParticleComponents 2))
  (bind _asfort_player_pawn (Utilities|Casting|CastToFortPlayerPawn MyTarget))
  (bind _output_get (Variables|Default|SetTargetPawn _asfort_player_pawn))
  (EventDispatchers|BindEventtoOnFootstepEvent _output_get (EventDispatchers|CreateEvent :!event OnPlayerFootstep))
  (EventDispatchers|BindEventtoOnPawnLanded _output_get (delegate Custom|OnPlayerLanded))
  (Pawn|BindEventtoOnPawnJumped _output_get (delegate Custom|OnPlayerJump))
  (Transformation|SetRelativeRotation _output_2 "0.000000,165.000000,0.000000" false false)
  (Transformation|SetRelativeLocation _output_2 "45.000000,0.000000,0.000000" false false)
  (Transformation|SetRelativeRotation _output "0.000000,-15.000000,0.000000" false false)
  (Transformation|SetRelativeLocation _output "-45.000000,0.000000,0.000000" false false)
  ; On Ice Audio Handling
  (Utilities|InstancedStruct|SetInstancedStructValue (Utilities|Array|Get(aref) (Variables|/Script/AudioMutators.AudioMutator|GetSelectors (Variables|Default|GetAudioMutatorFootstep))) (Utilities|Struct|MakeActorInstance (Variables|Default|GetTargetPawn)))
  (Audio|Mutator|AddMutator (AudioEngineSubsystems|GetAudioMutatorsSubsystem MyTarget) (Variables|Default|GetAudioMutatorFootstep))
  (SoundLibrary|AddLibrary (WorldSubsystems|GetSoundLibrarySubsystem (Variables|Default|GetTargetPawn)) (Variables|Default|GetTargetPawn) (Variables|Default|GetAdditiveSoundLibrary) :!class /Script/SoundLibrary.SoundLibrarySubsystem)
  (if (not (Utilities|Time|IsValidTimerHandle (Variables|Default|GetSlidingTimer)))
    (bind _returnvalue (Utilities|Time|SetTimerbyEvent (delegate Custom|Start%20Ice%20Sliding%20Audio) 0.1 true true))
    (Variables|Default|SetSlidingTimer _returnvalue)))

(event EventOnRemoval(EitherVFX-preferthisversion)
  (params (MyTarget :type Actor%20Object%20Reference) (Parameters :type Gameplay%20Cue%20Parameters%20Structure%20%28by%20ref%29) (ParticleComponents :type Array%20of%20FXSystem%20Component%20Object%20References) (AudioComponents :type Array%20of%20Audio%20Component%20Object%20References) (BurstCameraShakeInstance :type Camera%20Shake%20Base%20Object%20Reference) (BurstDecalInstance :type Decal%20Actor%20Object%20Reference))
  (bind _slidingtimer (Variables|Default|GetSlidingTimer))
  (Audio|Mutator|RemoveMutator (AudioEngineSubsystems|GetAudioMutatorsSubsystem MyTarget) (Variables|Default|GetAudioMutatorFootstep))
  (bind _asfort_player_pawn (Utilities|Casting|CastToFortPlayerPawn MyTarget))
  (EventDispatchers|UnbindEventfromOnFootstepEvent _asfort_player_pawn (EventDispatchers|CreateEvent :!event OnPlayerFootstep))
  (EventDispatchers|UnbindEventfromOnPawnLanded _asfort_player_pawn (delegate Custom|OnPlayerLanded))
  (Pawn|BindEventtoOnPawnJumped _asfort_player_pawn (delegate Custom|OnPlayerJump))
  (Effects|Components|ParticleSystem|SetBoolParameter (Utilities|Array|Get(acopy) (stale EventOnLoopingStart(EitherVFX-preferthisversion) ParticleComponents)) "Jump" true)
  (SoundLibrary|RemoveLibrary (WorldSubsystems|GetSoundLibrarySubsystem (Variables|Default|GetTargetPawn)) (Variables|Default|GetTargetPawn) "/MotoWrestleGameplay/Assets/Audio/IP_SL_MotoWrestle_Foley_FS_FreezingWater.IP_SL_MotoWrestle_Foley_FS_FreezingWater" :!class /Script/SoundLibrary.SoundLibrarySubsystem)
  (if (Utilities|Time|IsValidTimerHandle _slidingtimer)
    (Utilities|Time|ClearandInvalidateTimerbyHandle 0 _slidingtimer)
    (else
      _))
  (bind _iceslideaudiocomponent (Variables|Default|GetIceSlideAudioComponent)
    (:then
      ; On Ice Audio Handling
      (Audio|Components|Audio|FadeOut _iceslideaudiocomponent 0.3 0.0 "Linear" :!class /Script/Engine.AudioComponent))
    (:else)))

(event Custom|OnPlayerJump
  (Effects|Components|ParticleSystem|SetBoolParameter (Utilities|Array|Get(acopy) (stale EventOnLoopingStart(EitherVFX-preferthisversion) ParticleComponents)) "Jump" true))

(event Custom|OnPlayerLanded
  (params (Hit :type Hit%20Result%20Structure%20%28by%20ref%29))
  (Effects|Components|ParticleSystem|SetBoolParameter (Utilities|Array|Get(acopy) (stale EventOnLoopingStart(EitherVFX-preferthisversion) ParticleComponents)) "Jump" false))

(fn OnPlayerFootstep
  (params (FootstepIndex :type Integer))
  (bind _asplayer_pawn_athena (Utilities|Casting|CastToPlayerPawn_Athena (Variables|Default|GetTargetPawn))))

(fn SetMutatorActor
  (bind _returnvalue (Utilities|InstancedStruct|MakeInstancedStruct (Utilities|Struct|MakeActorInstance (Variables|Default|GetTargetPawn))))
  (Utilities|Array|Add (Variables|/Script/AudioMutators.AudioMutator|GetSelectors (Variables|Default|GetAudioMutatorFootstep)) _returnvalue))

(fn ConstructionScript)
